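# Sound cue for the Ashton Hippo alt impact in an Athena LTM, playing grapple shoot impact sounds.
# A Random node picks one of four equally weighted waves, and a Modulator applies a slight volume boost with small variation.
# It uses the impacts sound class, gadget-activate attenuation and rifle bullet impact concurrency to govern mixing, falloff and simultaneous plays.
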SoundCue /Game/Athena/Sounds/LTMs/Ashton/Hippo/Ashton_Hippo_Alt_Impact_Cue
root=Modulator_0
nodes [
 {id=Modulator_0, children [Random_1], properties {volumeMax=1.2, volumeMin=1.1}}
 {id=Random_1, children [WavePlayer_4,WavePlayer_5,WavePlayer_6,WavePlayer_7], properties {weights [1,1,1,1]}}
 {id=WavePlayer_4, properties {soundWaveAssetPtr=&/Ashton_Hippo_Grapple_Shoot_Impact_04}}
 {id=WavePlayer_5, properties {soundWaveAssetPtr=&/Ashton_Hippo_Grapple_Shoot_Impact_01}}
 {id=WavePlayer_6, properties {soundWaveAssetPtr=&/Ashton_Hippo_Grapple_Shoot_Impact_02}}
 {id=WavePlayer_7, properties {soundWaveAssetPtr=&/Ashton_Hippo_Grapple_Shoot_Impact_03}}
]
properties {
 attenuationSettings=&Fort_GadgetActivate_Att
 maxDistance=4000
 soundClassObject=&Fort_Impacts
 volumeMultiplier=1
 attenuationOverrides {bEnableReverbSend=false, falloffDistance=3000, nonSpatializedRadiusStart=1000, stereoSpread=2000, attenuationShapeExtents {x=2500}}
 concurrencySet [&RifleBulletImpact_Conc]
}
