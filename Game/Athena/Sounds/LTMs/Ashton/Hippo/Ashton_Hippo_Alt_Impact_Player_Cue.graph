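# Sound cue for the Ashton Hippo LTM's alternate player impact, built from grapple shoot impact recordings.
# It picks one of four equally weighted waves, applies a slight random volume boost and plays it after a short fixed delay.
# Overridden attenuation disables distance falloff, spatialization and reverb send, so it plays as non-positional audio.
# It routes through the impacts sound class and shares rifle bullet impact concurrency.
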SoundCue /Game/Athena/Sounds/LTMs/Ashton/Hippo/Ashton_Hippo_Alt_Impact_Player_Cue
root=Delay_0
nodes [
 {id=Delay_0, children [Modulator_0], properties {delayMax=0.05, delayMin=0.05}}
 {id=Modulator_0, children [Random_0], properties {volumeMax=1.2, volumeMin=1.1}}
 {id=Random_0, children [WavePlayer_0,WavePlayer_1,WavePlayer_2,WavePlayer_7], properties {weights [1,1,1,1]}}
 {id=WavePlayer_0, properties {soundWaveAssetPtr=&/Ashton_Hippo_Grapple_Shoot_Impact_04}}
 {id=WavePlayer_1, properties {soundWaveAssetPtr=&/Ashton_Hippo_Grapple_Shoot_Impact_01}}
 {id=WavePlayer_2, properties {soundWaveAssetPtr=&/Ashton_Hippo_Grapple_Shoot_Impact_02}}
 {id=WavePlayer_7, properties {soundWaveAssetPtr=&/Ashton_Hippo_Grapple_Shoot_Impact_03}}
]
properties {
 bOverrideAttenuation=true
 maxDistance=2097152
 soundClassObject=&Fort_Impacts
 volumeMultiplier=0.6
 attenuationOverrides {
  bAttenuate=false
  bEnableReverbSend=false
  bSpatialize=false
  falloffDistance=3000
  nonSpatializedRadiusStart=1000
  stereoSpread=2000
  attenuationShapeExtents {x=2500}
 }
 concurrencySet [&RifleBulletImpact_Conc]
}
