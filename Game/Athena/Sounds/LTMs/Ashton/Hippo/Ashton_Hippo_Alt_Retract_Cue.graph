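# Sound cue for the alternate grapple retract sound of the Hippo item in the Ashton limited-time mode.
# A Random node picks one of four equally weighted retract waves, and a Modulator varies its pitch slightly on top of a lowered base pitch.
# It uses grappling gadget attenuation, the gadgets/items sound class, and an owner-based two-voice concurrency rule that stops the oldest.
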SoundCue /Game/Athena/Sounds/LTMs/Ashton/Hippo/Ashton_Hippo_Alt_Retract_Cue
root=Modulator_1
nodes [
 {id=Modulator_1, children [Random_1], properties {pitchMax=1.1, pitchMin=0.9}}
 {id=Random_1, children [WavePlayer_0,WavePlayer_1,WavePlayer_2,WavePlayer_3], properties {weights [1,1,1,1]}}
 {id=WavePlayer_0, properties {soundWaveAssetPtr=&/Ashton_Hippo_Grapple_Retract_01}}
 {id=WavePlayer_1, properties {soundWaveAssetPtr=&/Ashton_Hippo_Grapple_Retract_02}}
 {id=WavePlayer_2, properties {soundWaveAssetPtr=&/Ashton_Hippo_Grapple_Retract_03}}
 {id=WavePlayer_3, properties {soundWaveAssetPtr=&/Ashton_Hippo_Grapple_Retract_04}}
]
properties {
 attenuationSettings=/Game/Sounds/Attenuation/Fort_GadgetGrappling_Att
 maxDistance=8500
 pitchMultiplier=0.9
 soundClassObject=&Fort_Gadgets_Items
 volumeMultiplier=1
 concurrencySet [/Game/Sounds/Concurrency_Rules/TwoMaximum_Owner_StopOldest]
}
