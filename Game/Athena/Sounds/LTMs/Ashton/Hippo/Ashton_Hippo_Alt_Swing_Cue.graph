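# Sound cue for the alternate swing of the Ashton Hippo grapple item, stored under Athena limited-time mode sounds.
# A Random node picks one of four equally weighted grapple swing waves, feeding a default Modulator for slight pitch and volume variation.
# It plays in the gadgets sound class under player melee weapon concurrency.
# Inline attenuation overrides, with reverb send off, supersede the referenced grappling gadget attenuation asset.
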SoundCue /Game/Athena/Sounds/LTMs/Ashton/Hippo/Ashton_Hippo_Alt_Swing_Cue
root=Modulator_2
nodes [
 {id=Modulator_2, children [Random_2]}
 {id=Random_2, children [WavePlayer_4,WavePlayer_5,WavePlayer_6,WavePlayer_7], properties {weights [1,1,1,1]}}
 {id=WavePlayer_4, properties {soundWaveAssetPtr=&/Ashton_Hippo_Grapple_Swing_01}}
 {id=WavePlayer_5, properties {soundWaveAssetPtr=&/Ashton_Hippo_Grapple_Swing_02}}
 {id=WavePlayer_6, properties {soundWaveAssetPtr=&/Ashton_Hippo_Grapple_Swing_03}}
 {id=WavePlayer_7, properties {soundWaveAssetPtr=&/Ashton_Hippo_Grapple_Swing_04}}
]
properties {
 attenuationSettings=/Game/Sounds/Attenuation/Fort_GadgetGrappling_Att
 bOverrideAttenuation=true
 maxDistance=3000
 soundClassObject=&Fort_Gadgets_Items
 volumeMultiplier=1
 attenuationOverrides {bEnableReverbSend=false, falloffDistance=3000, nonSpatializedRadiusStart=800, stereoSpread=2000, attenuationShapeExtents {x=0}}
 concurrencySet [&Player_Melee_Weapon_Conc]
}
