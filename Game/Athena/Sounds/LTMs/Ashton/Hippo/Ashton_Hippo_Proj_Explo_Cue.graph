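# Sound cue for the Ashton LTM Hippo projectile explosion, built from close, medium and distant layers.
# A mixer blends the three layers, each with its own attenuation, random wave pick and pitch modulation, with the distant layer mixed lower.
# The close layer adds a short fixed delay and volume variation, while the distant layer reuses explosive bow sounds.
# A concurrency override caps it at two instances, stopping the oldest.
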SoundCue /Game/Athena/Sounds/LTMs/Ashton/Hippo/Ashton_Hippo_Proj_Explo_Cue
root=Mixer_0
nodes [
 {
  id=Attenuation_0
  children [Modulator_2]
  properties {attenuationSettings=/Game/Sounds/Attenuation/Fort_Weapon_Arrow_Hippo_Expl_Med_Ahena_Att}
 }
 {
  id=Attenuation_1
  children [Modulator_0]
  properties {attenuationSettings=/Game/Sounds/Attenuation/Fort_Weapon_Arrow_Hippo_Dist_Expl_Athena_Att}
 }
 {
  id=Attenuation_7
  children [Modulator_6]
  properties {attenuationSettings=/Game/Sounds/Attenuation/Fort_Explosion_Close_StW_Att, attenuationOverrides {falloffDistance=1600}}
 }
 {id=Delay_0, children [Random_1], properties {delayMax=0.066, delayMin=0.066}}
 {id=Mixer_0, children [Attenuation_7, Attenuation_0, Attenuation_1], properties {inputVolume [1,1,0.7]}}
 {id=Modulator_0, children [Random_6], properties {pitchMax=1.1, pitchMin=0.9}}
 {id=Modulator_2, children [Random_2], properties {pitchMax=1.1, pitchMin=0.9}}
 {id=Modulator_6, children [Delay_0], properties {pitchMax=1.3, pitchMin=1, volumeMin=0.5}}
 {id=Random_1, children [WavePlayer_4,WavePlayer_5,WavePlayer_6,WavePlayer_7,WavePlayer_8,WavePlayer_9], properties {weights [1,1,1,1,1,1]}}
 {id=Random_2, children [WavePlayer_10, WavePlayer_11, WavePlayer_12, WavePlayer_13], properties {weights [1,1,1,1]}}
 {id=Random_6, children [WavePlayer_18, WavePlayer_19, WavePlayer_20, WavePlayer_21], properties {weights [1,1,1,1]}}
 {id=WavePlayer_10, properties {soundWaveAssetPtr=&/Ashton_Hippo_Explo_Med_01}}
 {id=WavePlayer_11, properties {soundWaveAssetPtr=&/Ashton_Hippo_Explo_Med_02}}
 {id=WavePlayer_12, properties {soundWaveAssetPtr=&/Ashton_Hippo_Explo_Med_03}}
 {id=WavePlayer_13, properties {soundWaveAssetPtr=&/Ashton_Hippo_Explo_Med_04}}
 {id=WavePlayer_18, properties {soundWaveAssetPtr=/Game/Sounds/Fort_Weapons/Explosive_Bow/Weapon_Bow_Compound_Arrow_Expl_Dist_01}}
 {id=WavePlayer_19, properties {soundWaveAssetPtr=/Game/Sounds/Fort_Weapons/Explosive_Bow/Weapon_Bow_Compound_Arrow_Expl_Dist_02}}
 {id=WavePlayer_20, properties {soundWaveAssetPtr=/Game/Sounds/Fort_Weapons/Explosive_Bow/Weapon_Bow_Compound_Arrow_Expl_Dist_03}}
 {id=WavePlayer_21, properties {soundWaveAssetPtr=/Game/Sounds/Fort_Weapons/Explosive_Bow/Weapon_Bow_Compound_Arrow_Expl_Dist_04}}
 {id=WavePlayer_4, properties {soundWaveAssetPtr=&/Ashton_Hippo_Explo_Close_01}}
 {id=WavePlayer_5, properties {soundWaveAssetPtr=&/Ashton_Hippo_Explo_Close_02}}
 {id=WavePlayer_6, properties {soundWaveAssetPtr=&/Ashton_Hippo_Explo_Close_03}}
 {id=WavePlayer_7, properties {soundWaveAssetPtr=&/Ashton_Hippo_Explo_Close_04}}
 {id=WavePlayer_8, properties {soundWaveAssetPtr=&/Ashton_Hippo_Explo_Close_05}}
 {id=WavePlayer_9, properties {soundWaveAssetPtr=&/Ashton_Hippo_Explo_Close_06}}
]
properties {
 bOverrideConcurrency=true
 maxDistance=32500
 soundClassObject=/Game/Sounds/FortSoundClassesAndModes/Fort_Explosives_Large_PSM
 volumeMultiplier=0.5661
 concurrencyOverrides {maxCount=2, resolutionRule=StopOldest}
}
