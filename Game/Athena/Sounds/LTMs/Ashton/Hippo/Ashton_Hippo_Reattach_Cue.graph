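# Sound cue for the reattach sound of "Hippo" in the Ashton limited-time mode, built from grappling hook reattach recordings.
# An equally weighted Random node picks one of three waves and feeds a root Modulator for variation.
# The cue plays slightly pitched up and quieter, with grappling gadget attenuation and the gadget items sound class.
# Its concurrency override refuses new instances once the limit is reached.
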SoundCue /Game/Athena/Sounds/LTMs/Ashton/Hippo/Ashton_Hippo_Reattach_Cue
root=Modulator_13
nodes [
 {id=Modulator_13, children [Random_0]}
 {id=Random_0, children [WavePlayer_0,WavePlayer_1,WavePlayer_2], properties {weights [1,1,1]}}
 {id=WavePlayer_0, properties {soundWaveAssetPtr=/Game/Sounds/Fort_Weapons/GrapplingHook/Grappling_Reattach_01}}
 {id=WavePlayer_1, properties {soundWaveAssetPtr=/Game/Sounds/Fort_Weapons/GrapplingHook/Grappling_Reattach_02}}
 {id=WavePlayer_2, properties {soundWaveAssetPtr=/Game/Sounds/Fort_Weapons/GrapplingHook/Grappling_Reattach_03}}
]
properties {
 attenuationSettings=/Game/Sounds/Attenuation/Fort_GadgetGrappling_Att
 bOverrideConcurrency=true
 maxDistance=8500
 pitchMultiplier=1.1
 soundClassObject=&Fort_Gadgets_Items
 volumeMultiplier=0.6
 concurrencyOverrides {resolutionRule=PreventNew}
}
