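# Sound cue for the reel-in sound of the Ashton Hippo item in a limited-time mode, judging by its name and path.
# A random node picks one of two equally weighted grappling hook retract waves.
# It plays at heavily reduced volume through the gadget items sound class, using the grappling gadget attenuation.
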SoundCue /Game/Athena/Sounds/LTMs/Ashton/Hippo/Ashton_Hippo_Reelin_Cue
root=Random_1
nodes [
 {id=Random_1, children [WavePlayer_1,WavePlayer_3], properties {weights [1,1]}}
 {id=WavePlayer_1, properties {soundWaveAssetPtr=/Game/Sounds/Fort_Weapons/GrapplingHook/Grappling_Retract_01}}
 {id=WavePlayer_3, properties {soundWaveAssetPtr=/Game/Sounds/Fort_Weapons/GrapplingHook/Grappling_Retract_02}}
]
properties {
 attenuationSettings=/Game/Sounds/Attenuation/Fort_GadgetGrappling_Att
 maxDistance=8500
 soundClassObject=&Fort_Gadgets_Items
 volumeMultiplier=0.1
}
