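# Sound cue for the reload sound of the Ashton LTM "Hippo" item, classed as weapon foley.
# A Random node picks one of three equally weighted reload wave variations.
# Playback uses the shared weapon reload attenuation at reduced volume, with a concurrency override and the debug flag enabled.
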SoundCue /Game/Athena/Sounds/LTMs/Ashton/Hippo/Ashton_Hippo_Reload_Cue
root=Random_0
nodes [
 {id=Random_0, children [WavePlayer_1,WavePlayer_2,WavePlayer_3], properties {weights [1,1,1]}}
 {id=WavePlayer_1, properties {soundWaveAssetPtr=&/Ashton_Hippo_Reload_01}}
 {id=WavePlayer_2, properties {soundWaveAssetPtr=&/Ashton_Hippo_Reload_02}}
 {id=WavePlayer_3, properties {soundWaveAssetPtr=&/Ashton_Hippo_Reload_03}}
]
properties {
 attenuationSettings=&Fort_Weapon_Reload_Att
 bDebug=true
 bOverrideConcurrency=true
 maxDistance=2250
 soundClassObject=&Fort_WeaponFoley
 volumeMultiplier=0.7
}
