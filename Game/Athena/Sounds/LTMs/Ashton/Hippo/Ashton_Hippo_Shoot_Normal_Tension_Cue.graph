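# Sound cue for the normal-tension shot of the Ashton Hippo weapon, layering close and distant firing sounds.
# The close layer randomly picks one of four equally weighted low-tension waves, using close projectile attenuation and the third-person ranged fire class.
# The distant layer reuses an Explosive Bow compound shot with distant bow attenuation and the distant fire class.
# A mixer combines both at reduced volume, and a concurrency override prevents new instances at the limit.
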SoundCue /Game/Athena/Sounds/LTMs/Ashton/Hippo/Ashton_Hippo_Shoot_Normal_Tension_Cue
root=Mixer_1
nodes [
 {id=Attenuation_3, children [Random_0], properties {attenuationSettings=&Fort_Weapon_Projectile_Close_Athena_Att}}
 {
  id=Attenuation_4
  children [WavePlayer_8]
  properties {attenuationSettings=/Game/Sounds/Attenuation/Fort_Weapon_Projectile_Distant_Bow_Athena_Att}
 }
 {id=Mixer_1, children [SoundClass_3,SoundClass_2], properties {inputVolume [0.6,0.6]}}
 {id=Random_0, children [WavePlayer_0,WavePlayer_1,WavePlayer_2,WavePlayer_7], properties {weights [1,1,1,1]}}
 {
  id=SoundClass_2
  children [Attenuation_4]
  properties {soundClassOverride=&Fort_WeaponRanged_Fire_Distant}
 }
 {
  id=SoundClass_3
  children [Attenuation_3]
  properties {soundClassOverride=&Fort_WeaponRanged_Fire_3P_PSM}
 }
 {id=WavePlayer_0, properties {soundWaveAssetPtr=&/Ashton_Hippo_Shoot_Low_Tension_01}}
 {id=WavePlayer_1, properties {soundWaveAssetPtr=&/Ashton_Hippo_Shoot_Low_Tension_02}}
 {id=WavePlayer_2, properties {soundWaveAssetPtr=&/Ashton_Hippo_Shoot_Low_Tension_03}}
 {id=WavePlayer_7, properties {soundWaveAssetPtr=&/Ashton_Hippo_Shoot_Low_Tension_04}}
 {id=WavePlayer_8, properties {soundWaveAssetPtr=/Game/Sounds/Fort_Weapons/Explosive_Bow/Weapon_Bow_Compound_Shoot_Med_01}}
]
properties {
 bOverrideConcurrency=true
 maxDistance=27500
 volumeMultiplier=1
 attenuationOverrides {falloffDistance=5000, attenuationShapeExtents {x=1000}}
 concurrencyOverrides {resolutionRule=PreventNew}
}
