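# Sound cue for the Hippo close whoosh-by effect in the Ashton LTM, classed as a ranged weapon projectile effect.
# An equally weighted Random node picks one of seven wave variations and feeds a Modulator that holds pitch fixed.
# Overridden attenuation disables distance volume falloff and reverb send, and enables complex-collision occlusion that silences the sound when blocked.
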SoundCue /Game/Athena/Sounds/LTMs/Ashton/Hippo/Ashton_Hippo_Whoosh_By_Close_Cue
root=Modulator_0
nodes [
 {id=Modulator_0, children [Random_1], properties {pitchMax=1, pitchMin=1}}
 {
  id=Random_1
  children [WavePlayer_0, WavePlayer_5, WavePlayer_6, WavePlayer_7, WavePlayer_8, WavePlayer_9, WavePlayer_10]
  properties {weights [1,1,1,1,1,1,1]}
 }
 {id=WavePlayer_0, properties {soundWaveAssetPtr=&/Ashton_Hippo_Whoosh_By_Close_01}}
 {id=WavePlayer_10, properties {soundWaveAssetPtr=&/Ashton_Hippo_Whoosh_By_Close_07}}
 {id=WavePlayer_5, properties {soundWaveAssetPtr=&/Ashton_Hippo_Whoosh_By_Close_02}}
 {id=WavePlayer_6, properties {soundWaveAssetPtr=&/Ashton_Hippo_Whoosh_By_Close_03}}
 {id=WavePlayer_7, properties {soundWaveAssetPtr=&/Ashton_Hippo_Whoosh_By_Close_04}}
 {id=WavePlayer_8, properties {soundWaveAssetPtr=&/Ashton_Hippo_Whoosh_By_Close_05}}
 {id=WavePlayer_9, properties {soundWaveAssetPtr=&/Ashton_Hippo_Whoosh_By_Close_06}}
]
properties {
 attenuationSettings=/Game/Sounds/Attenuation/Fort_Weapon_Projectile_InAir_Loop_Att
 bDebug=true
 bOverrideAttenuation=true
 bOverrideConcurrency=true
 maxDistance=2097152
 soundClassObject=&Fort_WeaponRanged_Projectile_Effects
 volumeMultiplier=1
 attenuationOverrides {
  bAttenuate=false
  bEnableOcclusion=true
  bEnableReverbSend=false
  bUseComplexCollisionForOcclusion=true
  falloffDistance=0
  occlusionTraceChannel=ECC_GameTraceChannel11
  occlusionVolumeAttenuation=0
  attenuationShapeExtents {x=50000}
 }
}
