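# Sound cue for a medium whoosh-by in the Ashton LTM's Hippo sound set, classed as a ranged weapon projectile effect.
# A Random node picks among six equally weighted wave variations, played at reduced volume.
# Overridden attenuation supersedes the assigned projectile in-air loop asset, disabling distance attenuation and reverb send while enabling complex-collision occlusion tracing.
# Concurrency is also overridden on the cue, and its debug flag is left on.
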SoundCue /Game/Athena/Sounds/LTMs/Ashton/Hippo/Ashton_Hippo_Whoosh_By_Med_Cue
root=Random_0
nodes [
 {id=Random_0, children [WavePlayer_1,WavePlayer_2,WavePlayer_3,WavePlayer_4,WavePlayer_8,WavePlayer_9], properties {weights [1,1,1,1,1,1]}}
 {id=WavePlayer_1, properties {soundWaveAssetPtr=&/Ashton_Hippo_Whoosh_By_Med_01}}
 {id=WavePlayer_2, properties {soundWaveAssetPtr=&/Ashton_Hippo_Whoosh_By_Med_02}}
 {id=WavePlayer_3, properties {soundWaveAssetPtr=&/Ashton_Hippo_Whoosh_By_Med_03}}
 {id=WavePlayer_4, properties {soundWaveAssetPtr=&/Ashton_Hippo_Whoosh_By_Med_04}}
 {id=WavePlayer_8, properties {soundWaveAssetPtr=&/Ashton_Hippo_Whoosh_By_Med_05}}
 {id=WavePlayer_9, properties {soundWaveAssetPtr=&/Ashton_Hippo_Whoosh_By_Med_06}}
]
properties {
 attenuationSettings=/Game/Sounds/Attenuation/Fort_Weapon_Projectile_InAir_Loop_Att
 bDebug=true
 bOverrideAttenuation=true
 bOverrideConcurrency=true
 maxDistance=2097152
 soundClassObject=&Fort_WeaponRanged_Projectile_Effects
 volumeMultiplier=0.4
 attenuationOverrides {
  bAttenuate=false
  bEnableOcclusion=true
  bEnableReverbSend=false
  bUseComplexCollisionForOcclusion=true
  falloffDistance=1000
  occlusionTraceChannel=ECC_GameTraceChannel11
  occlusionVolumeAttenuation=0
  attenuationShapeExtents {x=50000}
 }
}
